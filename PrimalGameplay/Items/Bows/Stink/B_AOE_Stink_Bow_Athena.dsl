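; Area-of-effect stink cloud actor for the Stink Bow, with a delayed effect area and a timed lifetime.
; On BeginPlay it disables collision on the overlap area, fades in looping audio, and enables query-only collision after the damage delay.
; A second timer, set to cloud duration plus that delay, fades out the audio, deactivates the Niagara effect and kills the delivery.
; Delay and duration come from scalable float rows.
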(event EventBeginPlay
  (bind _row_damagedelay (Variables|Default|GetRow_DamageDelay))
  (bind _returnvalue (ScalableFloat|GetValueatLevel _row_damagedelay))
  (bind _effectoverlaparea (Variables|Default|GetEffectOverlapArea))
  (Collision|SetCollisionComponent _effectoverlaparea)
  (Collision|SetCollisionEnabled _effectoverlaparea "NoCollision")
  (Utilities|Time|SetTimerbyFunctionName 0 "TurnOnCollision" _returnvalue)
  (Utilities|Time|SetTimerbyEvent (delegate Custom|KillCloud) (+ (ScalableFloat|GetValueatLevel (Variables|Default|GetRow_CloudDuration)) _returnvalue))
  (Audio|Components|Audio|Play (Variables|Default|GetLoopingAudio) 0.0)
  (CallFunction|FadeIn))

(event Custom|KillCloud
  (CallFunction|FadeOut)
  (Delivery|Kill))

(event Custom|FadeIn
  ;[1 1 1] Fade in/out audio and VFX
  (Audio|Components|Audio|FadeIn (Variables|Default|GetLoopingAudio) (ScalableFloat|GetValueatLevel (Variables|Default|GetRow_DamageDelay)) 1.0 0.0 "Linear" :!class /Script/Engine.AudioComponent))

(event Custom|FadeOut
  ;[1 1 1] Fade in/out audio and VFX
  (Audio|Components|Audio|FadeOut (Variables|Default|GetLoopingAudio) (Variables|Default|GetCloudFadeOutTime) 0.0 "Linear" :!class /Script/Engine.AudioComponent)
  (Components|Activation|Deactivate (Variables|Default|GetNS_Bow_Stink_Bomb_AoE)))

(event Custom|TurnOnCollision
  (Collision|SetCollisionEnabled (Variables|Default|GetEffectOverlapArea) "QueryOnly"))

(fn ConstructionScript)
