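; Arrow projectile Blueprint for the stink bow, which leaves a stink cloud where a fully charged shot lands.
; The fully charged impact event breaks the hit result and spawns the stink cloud building gameplay actor class at the hit location, passing the projectile's instigator.
; The construction script only calls its parent.
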(event EventOnFullyChargedImpact
  (params (Hit%20Result :type Hit%20Result%20Structure))
  (bind (_bblockinghit _binitialoverlap _time _distance _location) (Collision|BreakHitResult Hit%20Result))
  (Building|SpawnBuildingGameplayActor (Variables|Default|GetStinkCloudBGAClass) (Math|Transform|MakeTransform _location) (Game|GetInstigator)))

(fn ConstructionScript
  (|CallParentFunction))
